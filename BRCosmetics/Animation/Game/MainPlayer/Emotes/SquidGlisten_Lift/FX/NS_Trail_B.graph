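# Niagara trail effect for the SquidGlisten_Lift emote, drawing red-to-purple ribbons behind an emote attach bone.
# A burst-spawned Position particle follows attach_emote bones on the parent mesh, with User.Anim_Curve_Var selecting the index.
# Base, Outline and Streak ribbons sample that particle, with animation-curve user floats driving spawn rate and initial opacity, tinted by time of day.
# Floored emitter age works as a step switch for ribbon facing and delays Streak visibility.
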NiagaraSystem /BRCosmetics/Animation/Game/MainPlayer/Emotes/SquidGlisten_Lift/FX/NS_Trail_B
dependencies {
 usedDataInterfaces [
  &NiagaraDataInterfaceParticleRead
  &NiagaraDataInterfaceColorCurve
  &NiagaraDataInterfaceCurve
  &NiagaraDataInterfaceSkeletalMesh
  &NiagaraDataInterfaceArrayFloat
 ]
 usedDynamicInputs [&SelectFloatFromArray]
 usedModules [
  &SystemState
  &self:SystemSpawnScript.SystemScriptSource.SystemScriptGraph.NiagaraNodeAssignment_0.SetVariables_#1
 ]
}
emitters {
 Position {
  emitterUpdateScript [
   &EmitterState(Life Cycle Mode=::NewEnumerator1, Loop Duration=15, UseLoopDelay=false)
   &SpawnBurst_Instantaneous(Use Spawn Probability=false, UseLoopCountLimit=false)
  ]
  particleSpawnScript [&InitializeParticle(
    Color=<rgba 0.3542 0.6261 1 0.2>
    Lifetime=15
    Sprite Size Mode=::NewEnumerator3
    Uniform Sprite Size=30
    UsePositionOffset=false
    Write Material Random=true
   )]
  particleUpdateScript [
   &ParticleState()
   &SkeletalMeshLocation(
    Bone / Socket Index=&MakeIntFromFloat(Float=$User.Anim_Curve_Var, Conversion Mode=::NewEnumerator1)
    Bone Sampling Mode=::NewEnumerator4
    MaskSpawnGroup=false
    Orientation Sampling=::NewEnumerator1
    Skeletal Mesh=DataInterface() {propertyValues {
      DefaultMesh=null
      ExcludeBoneName=root
      OptionalComponentClass=null
      PreviewMesh=null
      SkinningMode=SkinOnTheFly
      SoftSourceActor=null
      SourceMode=AttachParent
      UvSetIndex=0
      WholeMeshLOD=-1
      bExcludeBone=false
      bReadDeformedGeometry=true
      bRequireCurrentFrameData=true
      ComponentTags []
      FilteredBones [attach_emote_02, attach_emote_01]
      FilteredSockets []
      MeshUserParameter {parameter {name=None, typeDefHandle {registeredTypeIndex=1}, varData []}}
      SamplingRegions []
     }}
   ) {module=SkeletalMeshLocation001}
  ]
  rendererClasses [&NiagaraSpriteRendererProperties]
  renderers [{rendererClass=&NiagaraSpriteRendererProperties}]
 }
 Red_Base {
  emitterUpdateScript [
   &EmitterState(Life Cycle Mode=::NewEnumerator1, Loop Delay=1, Loop Duration=15, UseLoopDelay=false)
   &SpawnRate(SpawnRate=&Multiply_Float(A=50, B=$User.Anim_Curve_SpawnRate), Use Spawn Probability=false)
   &SpawnParticlesFromOtherEmitter(
    Attribute Reader=DataInterface() {propertyValues {EmitterBinding {bindingMode=Other, emitterName=Position}}}
    Spawn Rate=0
    Use Maximum Age=false
    Use Minimum Age=false
   )
  ]
  particleSpawnScript [
   &InitializeParticle(
    Color=&MultiplyLinearColor(
     Linear Color=&FNTimeOfDayColor(Night Color=<rgba 1 1 1 0.25>, Morning Color=<rgba 1 1 1 0.85>, Day Color=<rgba 1 1 1 1>, Evening Color=<rgba 1 1 1 0.85>)
     Scale Factor=&Vector4FromCurve(
      CurveIndex=$User.Anim_Curve_InitialOpacity
      Vector4Curve=DataInterface() {propertyValues {
        AlphaCurve {keys [<at 0 0>,<at 1 1>]}
        BlueCurve {keys [<at 0 1>,<at 1 1>]}
        GreenCurve {keys [<at 0 1>,<at 1 1>]}
        RedCurve {keys [<at 0 1>,<at 1 1>]}
       }}
     )
    )
    Lifetime=0.45
    Ribbon Width=22
    Ribbon Width Mode=::NewEnumerator1
    UsePositionOffset=false
    Write Material Random=true
   )
   &SampleParticlesFromOtherEmitter()
   /SetVariables(Particles.RibbonFacing=&Lerp_Vector(A=<0 1 0>, B=<0 0 1>, Alpha=/Niagara/DynamicInputs/Math/Floor_Float(Float:NiagaraFloat=$Emitter.Age))) {isSetParameters=true}
  ]
  particleUpdateScript [
   &ParticleState()
   &SolveForcesAndVelocity(Clamp Velocity=false, Limit Acceleration=false)
   &ScaleRibbonWidth(Ribbon Width Scale=&FloatFromCurve(FloatCurve=DataInterface() {propertyValues {Curve {keys [<at 0 1>,<at 1 0>]}}})) {module=ScaleRibbonWidth001}
   &ScaleColor(
    Linear Color Curve=DataInterface() {propertyValues {
      AlphaCurve {keys [<at 0 0>,<at 0.108 1>,<at 1 0>]}
      BlueCurve {keys [<at 0 0.0306>,<at 0.2385 0.0478>,<at 0.8625 0.5156>,<at 1 0.2083>]}
      GreenCurve {keys [<at 0 0.0248>,<at 0.2385 0>,<at 0.8625 0>,<at 1 0.0083>]}
      RedCurve {keys [<at 0 0.9531>,<at 0.2385 0.625>,<at 0.8625 0.3372>,<at 1 0>]}
     }}
    Scale Mode=::NewEnumerator2
   ) {module=ScaleColor001}
   &ScaleColor(Scale Alpha=0.4, ScaleA=true, ScaleRGB=true)
  ]
  rendererClasses [&NiagaraRibbonRendererProperties]
  renderers [{rendererClass=&NiagaraRibbonRendererProperties}]
 }
 Red_Outline {
  emitterUpdateScript [
   &EmitterState(Life Cycle Mode=::NewEnumerator1, Loop Delay=1, Loop Duration=15, UseLoopDelay=false)
   &SpawnRate(SpawnRate=&Multiply_Float(A=50, B=$User.Anim_Curve_SpawnRate), Use Spawn Probability=false)
   &SpawnParticlesFromOtherEmitter(
    Attribute Reader=DataInterface() {propertyValues {EmitterBinding {bindingMode=Other, emitterName=Position}}}
    Spawn Rate=0
    Use Maximum Age=false
    Use Minimum Age=false
   )
  ]
  particleSpawnScript [
   &InitializeParticle(
    Color=&MultiplyLinearColor(
     Linear Color=&FNTimeOfDayColor(Night Color=<rgba 1 1 1 0.25>, Morning Color=<rgba 1 1 1 0.85>, Day Color=<rgba 1 1 1 1>, Evening Color=<rgba 1 1 1 0.85>)
     Scale Factor=&Vector4FromCurve(
      CurveIndex=$User.Anim_Curve_InitialOpacity
      Vector4Curve=DataInterface() {propertyValues {
        AlphaCurve {keys [<at 0 0>,<at 1 1>]}
        BlueCurve {keys [<at 0 1>,<at 1 1>]}
        GreenCurve {keys [<at 0 1>,<at 1 1>]}
        RedCurve {keys [<at 0 1>,<at 1 1>]}
       }}
     )
    )
    Lifetime=0.45
    Ribbon Width=25
    Ribbon Width Mode=::NewEnumerator1
    UsePositionOffset=false
    Write Material Random=true
   )
   &SampleParticlesFromOtherEmitter()
   /SetVariables(Particles.RibbonFacing=&Lerp_Vector(A=<0 1 0>, B=<0 0 1>, Alpha=&Floor_Float(Float:NiagaraFloat=$Emitter.Age))) {isSetParameters=true}
  ]
  particleUpdateScript [
   &ParticleState()
   &SolveForcesAndVelocity(Clamp Velocity=false, Limit Acceleration=false)
   &ScaleRibbonWidth(Ribbon Width Scale=&FloatFromCurve(FloatCurve=DataInterface() {propertyValues {Curve {keys [<at 0 1>,<at 1 0>]}}})) {module=ScaleRibbonWidth001}
   &ScaleColor(
    Linear Color Curve=DataInterface() {propertyValues {
      AlphaCurve {keys [<at 0 0>,<at 0.108 1>,<at 1 0>]}
      BlueCurve {keys [<at 0 0.0306>,<at 0.1769 0.0478>,<at 0.8625 0.5156>,<at 1 0.2083>]}
      GreenCurve {keys [<at 0 0.0248>,<at 0.1769 0>,<at 0.8625 0>,<at 1 0.0083>]}
      RedCurve {keys [<at 0 0.9531>,<at 0.1769 0.625>,<at 0.8625 0.3372>,<at 1 0>]}
     }}
    Scale Mode=::NewEnumerator2
   )
   &ScaleColor(Scale Alpha=0.2, ScaleA=true, ScaleRGB=true) {module=ScaleColor002}
  ]
  rendererClasses [&NiagaraRibbonRendererProperties]
  renderers [{rendererClass=&NiagaraRibbonRendererProperties}]
 }
 Red_Streak {
  emitterUpdateScript [
   &EmitterState(Life Cycle Mode=::NewEnumerator1, Loop Behavior=::NewEnumerator1, Loop Delay=1, Loop Duration=15, UseLoopDelay=false)
   &SpawnRate(SpawnRate=&Multiply_Float(A=50, B=$User.Anim_Curve_SpawnRate), Use Spawn Probability=false)
   &SpawnParticlesFromOtherEmitter(
    Attribute Reader=DataInterface() {propertyValues {EmitterBinding {bindingMode=Other, emitterName=Position}}}
    Spawn Rate=0
    Use Maximum Age=false
    Use Minimum Age=false
   )
  ]
  particleSpawnScript [
   &InitializeParticle(
    Color=&MultiplyLinearColor(
     Linear Color=&FNTimeOfDayColor(Night Color=<rgba 1 1 1 0.25>, Morning Color=<rgba 1 1 1 0.85>, Day Color=<rgba 1 1 1 1>, Evening Color=<rgba 1 1 1 0.85>)
     Scale Factor=&Vector4FromCurve(
      CurveIndex=$User.Anim_Curve_InitialOpacity
      Vector4Curve=DataInterface() {propertyValues {
        AlphaCurve {keys [<at 0 0>,<at 1 2>]}
        BlueCurve {keys [<at 0 1>,<at 1 1>]}
        GreenCurve {keys [<at 0 1>,<at 1 1>]}
        RedCurve {keys [<at 0 1>,<at 1 1>]}
       }}
     )
    )
    Lifetime=0.45
    Ribbon Width=25
    Ribbon Width Mode=::NewEnumerator1
    UsePositionOffset=false
    Write Material Random=true
   )
   &SampleParticlesFromOtherEmitter()
   /SetVariables(Particles.RibbonFacing=&Lerp_Vector(A=<0 1 0>, B=<0 0 1>, Alpha=&Floor_Float(Float:NiagaraFloat=$Emitter.Age))) {isSetParameters=true}
  ]
  particleUpdateScript [
   &ParticleState()
   &SolveForcesAndVelocity(Clamp Velocity=false, Limit Acceleration=false)
   &ScaleRibbonWidth(Ribbon Width Scale=&FloatFromCurve(FloatCurve=DataInterface() {propertyValues {Curve {keys [<at 0 1>,<at 1 0>]}}})) {module=ScaleRibbonWidth001}
   &ScaleColor(
    Linear Color Curve=DataInterface() {propertyValues {
      AlphaCurve {keys [<at 0 0>,<at 0.108 1>,<at 1 0>]}
      BlueCurve {keys [<at 0 0.0306>,<at 0.1769 0.0478>,<at 0.8625 0.5156>,<at 1 0.2083>]}
      GreenCurve {keys [<at 0 0.0248>,<at 0.1769 0>,<at 0.8625 0>,<at 1 0.0083>]}
      RedCurve {keys [<at 0 0.9531>,<at 0.1769 0.625>,<at 0.8625 0.3372>,<at 1 0>]}
     }}
    Scale Mode=::NewEnumerator2
   )
   &ScaleColor(
    Scale Alpha=&Lerp_Float(
     A=0
     B=0.4
     Alpha=&Floor_Float(Float:NiagaraFloat=&Divide_Float(A=$Emitter.Age, B=4))
    )
    ScaleA=true
    ScaleRGB=true
   ) {module=ScaleColor002}
  ]
  rendererClasses [&NiagaraRibbonRendererProperties]
  renderers [{rendererClass=&NiagaraRibbonRendererProperties}]
 }
}
summary {
 systemName=NS_Trail_B
 emitters [
  {emitterName=Red_Base, rendererClasses [&NiagaraRibbonRendererProperties]}
  {emitterName=Position, rendererClasses [&NiagaraSpriteRendererProperties]}
  {emitterName=Red_Outline, rendererClasses [&NiagaraRibbonRendererProperties]}
  {emitterName=Red_Streak, rendererClasses [&NiagaraRibbonRendererProperties]}
 ]
 userVariables [
  User.Anim_Curve_InitialOpacity {type=NiagaraFloat, defaultValue=1, description=""}
  User.Anim_Curve_SmallTrail_InitialOpacity {type=NiagaraFloat, defaultValue=0, description=""}
  User.Anim_Curve_SpawnRate {type=NiagaraFloat, defaultValue=0, description=""}
  User.Anim_Curve_Var {type=NiagaraFloat, defaultValue=0, description=""}
  User.bIsFrontEnd {type=NiagaraBool}
  User.bIsFrontEndPreview {type=NiagaraBool}
  User.FX_WeaponCaseSize {type=NiagaraInt32, defaultValue=1}
  User.NormalizedNotifyProgress {type=NiagaraFloat, defaultValue=0, description=""}
  User.SystemScale {type=NiagaraFloat, defaultValue=1, description=""}
 ]
}
